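# Sound cue for the Deimos ranged monster's ice-fire attack, named as a fire sound and using projectile-launch attenuation.
# A random node picks with equal weight among six Sideways Brute energy-ball throw waves from an S18 reskin folder.
# It plays in the enemy attack sound class at slightly reduced volume, with a concurrency override capping simultaneous instances.
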SoundCue /Game/Sounds/Fort_Monster_Deimos/Ranged/DeimosRanged_IceFire_Cue
root=Random_0
nodes [
 {
  id=Random_0
  children [WavePlayer_0,WavePlayer_2,WavePlayer_3,WavePlayer_4,WavePlayer_5,WavePlayer_6]
  properties {bShouldExcludeFromBranchCulling=true, weights [1,1,1,1,1,1]}
 }
 {id=WavePlayer_0, properties {soundWaveAssetPtr=/Game/Sounds/Fort_Monster_Deimos/SFX/S18_Reskin/Sideways_Brute_EnergyBall_Throw_01}}
 {id=WavePlayer_2, properties {soundWaveAssetPtr=/Game/Sounds/Fort_Monster_Deimos/SFX/S18_Reskin/Sideways_Brute_EnergyBall_Throw_02}}
 {id=WavePlayer_3, properties {soundWaveAssetPtr=/Game/Sounds/Fort_Monster_Deimos/SFX/S18_Reskin/Sideways_Brute_EnergyBall_Throw_03}}
 {id=WavePlayer_4, properties {soundWaveAssetPtr=/Game/Sounds/Fort_Monster_Deimos/SFX/S18_Reskin/Sideways_Brute_EnergyBall_Throw_04}}
 {id=WavePlayer_5, properties {soundWaveAssetPtr=/Game/Sounds/Fort_Monster_Deimos/SFX/S18_Reskin/Sideways_Brute_EnergyBall_Throw_05}}
 {id=WavePlayer_6, properties {soundWaveAssetPtr=/Game/Sounds/Fort_Monster_Deimos/SFX/S18_Reskin/Sideways_Brute_EnergyBall_Throw_06}}
]
properties {
 attenuationSettings=/Game/Sounds/Attenuation/Fort_Deimos_Prj_Launch_Att
 bOverrideConcurrency=true
 maxDistance=14000
 soundClassObject=/Game/Sounds/FortSoundClassesAndModes/Fort_Enemy_Attack
 volumeMultiplier=0.825
 concurrencyOverrides {maxCount=3}
}
