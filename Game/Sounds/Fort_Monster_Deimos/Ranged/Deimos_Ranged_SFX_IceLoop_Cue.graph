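# Sound cue for the looping in-flight sound of a Deimos monster's ranged ice projectile, using reskinned brute energy ball loop waves.
# A Random node picks one of two equally weighted loops, and a Modulator randomly raises its pitch.
# An Enveloper fades the volume in before a strengthened Doppler node shifts pitch with movement.
# Projectile in-air attenuation and a concurrency override capping simultaneous instances round out the cue.
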SoundCue /Game/Sounds/Fort_Monster_Deimos/Ranged/Deimos_Ranged_SFX_IceLoop_Cue
root=Doppler_0
nodes [
 {id=Doppler_0, children [Enveloper_0], properties {dopplerIntensity=3}}
 {
  id=Enveloper_0
  children [Modulator_1]
  properties {pitchCurve {editorCurveData {keys [{time=0}]}}, volumeCurve {editorCurveData {keys [<at 0 0.01 RCIM_Cubic>,<at 0.5 1>]}}}
 }
 {id=Modulator_1, children [Random_0], properties {pitchMax=1.25, pitchMin=1}}
 {id=Random_0, children [WavePlayer_0,WavePlayer_2], properties {bShouldExcludeFromBranchCulling=true, weights [1,1]}}
 {id=WavePlayer_0, properties {soundWaveAssetPtr=/Game/Sounds/Fort_Monster_Deimos/SFX/S18_Reskin/Sideways_Brute_EnergyBall_Loop_01}}
 {id=WavePlayer_2, properties {soundWaveAssetPtr=/Game/Sounds/Fort_Monster_Deimos/SFX/S18_Reskin/Sideways_Brute_EnergyBall_Loop_02}}
]
properties {
 attenuationSettings=/Game/Sounds/Attenuation/Fort_Deimos_Prj_InAir_Loop_Att
 bOverrideConcurrency=true
 maxDistance=3300
 soundClassObject=&Fort_WeaponRanged_Projectile_Effects
 volumeMultiplier=0.5
 attenuationOverrides {falloffDistance=3400, attenuationShapeExtents {x=1000}}
 concurrencyOverrides {maxCount=3}
}
