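# Sound cue for the ice projectile impact of the Deimos ranged monster, using reskinned Sideways Brute energy ball impact waves.
# A Random node picks one of four equally weighted variations and is excluded from branch culling, feeding a Modulator that slightly raises pitch at random.
# It plays through a small projectile impact attenuation, the ranged projectile effects sound class and a three-voice, stop-furthest concurrency set.
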SoundCue /Game/Sounds/Fort_Monster_Deimos/Ranged/Deimos_Ranged_SFX_Ice_Impact_Cue
root=Modulator_0
nodes [
 {id=Modulator_0, children [Random_0], properties {pitchMax=1.1, pitchMin=1}}
 {
  id=Random_0
  children [WavePlayer_0,WavePlayer_1,WavePlayer_2,WavePlayer_3]
  properties {bShouldExcludeFromBranchCulling=true, weights [1,1,1,1]}
 }
 {id=WavePlayer_0, properties {soundWaveAssetPtr=/Game/Sounds/Fort_Monster_Deimos/SFX/S18_Reskin/Sideways_Brute_EnergyBall_Impact_01}}
 {id=WavePlayer_1, properties {soundWaveAssetPtr=/Game/Sounds/Fort_Monster_Deimos/SFX/S18_Reskin/Sideways_Brute_EnergyBall_Impact_02}}
 {id=WavePlayer_2, properties {soundWaveAssetPtr=/Game/Sounds/Fort_Monster_Deimos/SFX/S18_Reskin/Sideways_Brute_EnergyBall_Impact_03}}
 {id=WavePlayer_3, properties {soundWaveAssetPtr=/Game/Sounds/Fort_Monster_Deimos/SFX/S18_Reskin/Sideways_Brute_EnergyBall_Impact_04}}
]
properties {
 attenuationSettings=/Game/Sounds/Attenuation/Fort_Deimos_Prj_Impact_Small_Att
 maxDistance=3750
 priority=0.5
 soundClassObject=&Fort_WeaponRanged_Projectile_Effects
 volumeMultiplier=1
 concurrencyOverrides {resolutionRule=PreventNew}
 concurrencySet [/Game/Sounds/Concurrency_Rules/ThreeMaximum_StopFurthest]
}
