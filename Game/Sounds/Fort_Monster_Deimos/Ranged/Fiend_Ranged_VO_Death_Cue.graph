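# Sound cue for the death vocalization of the ranged Deimos Fiend monster, in the tactical creature vocalization sound class.
# A mixer layers a close branch of random ranged death waves with a slightly quieter distant branch of random Fiend distant pain waves.
# Each branch has its own modulator and close or distant attenuation asset.
# Global and local Deimos vocal concurrency sets apply, and virtualization is disabled.
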SoundCue /Game/Sounds/Fort_Monster_Deimos/Ranged/Fiend_Ranged_VO_Death_Cue
root=Mixer_0
nodes [
 {id=Attenuation_0, children [Modulator_0], properties {attenuationSettings=/Game/Sounds/Attenuation/Fort_Deimos_Vocal_Close_Att}}
 {id=Attenuation_1, children [Modulator_1], properties {attenuationSettings=/Game/Sounds/Attenuation/Fort_Deimos_Vocal_Distant_Att}}
 {id=Mixer_0, children [Attenuation_0, Attenuation_1], properties {inputVolume [1,0.8]}}
 {id=Modulator_0, children [Random_2], properties {pitchMax=1.2}}
 {id=Modulator_1, children [Random_0]}
 {
  id=Random_0
  children [WavePlayer_0, WavePlayer_1, WavePlayer_2, WavePlayer_3, WavePlayer_11, WavePlayer_12]
  properties {bShouldExcludeFromBranchCulling=true, weights [1,1,1,1,1,1]}
 }
 {
  id=Random_2
  children [WavePlayer_13, WavePlayer_14, WavePlayer_15, WavePlayer_16, WavePlayer_17, WavePlayer_18, WavePlayer_19]
  properties {bShouldExcludeFromBranchCulling=true, weights [1,1,1,1,1,1,1]}
 }
 {id=WavePlayer_0, properties {soundWaveAssetPtr=/Game/Sounds/Fort_Monster_Deimos/Fiend/Fiend_Pain_New_Distant_01}}
 {id=WavePlayer_1, properties {soundWaveAssetPtr=/Game/Sounds/Fort_Monster_Deimos/Fiend/Fiend_Pain_New_Distant_02}}
 {id=WavePlayer_11, properties {soundWaveAssetPtr=/Game/Sounds/Fort_Monster_Deimos/Fiend/Fiend_Pain_New_Distant_06}}
 {id=WavePlayer_12, properties {soundWaveAssetPtr=/Game/Sounds/Fort_Monster_Deimos/Fiend/Fiend_Pain_New_Distant_07}}
 {id=WavePlayer_13, properties {soundWaveAssetPtr=&/Fiend_Ranged_Death_01}}
 {id=WavePlayer_14, properties {soundWaveAssetPtr=&/Fiend_Ranged_Death_02}}
 {id=WavePlayer_15, properties {soundWaveAssetPtr=&/Fiend_Ranged_Death_03}}
 {id=WavePlayer_16, properties {soundWaveAssetPtr=&/Fiend_Ranged_Death_04}}
 {id=WavePlayer_17, properties {soundWaveAssetPtr=&/Fiend_Ranged_Death_05}}
 {id=WavePlayer_18, properties {soundWaveAssetPtr=&/Fiend_Ranged_Death_06}}
 {id=WavePlayer_19, properties {soundWaveAssetPtr=&/Fiend_Ranged_Death_07}}
 {id=WavePlayer_2, properties {soundWaveAssetPtr=/Game/Sounds/Fort_Monster_Deimos/Fiend/Fiend_Pain_New_Distant_03}}
 {id=WavePlayer_3, properties {soundWaveAssetPtr=/Game/Sounds/Fort_Monster_Deimos/Fiend/Fiend_Pain_New_Distant_04}}
]
properties {
 bBypassVolumeScaleForPriority=true
 maxDistance=24000
 priority=0.75
 soundClassObject=&Fort_Creatures_Vocalization_Tactical
 virtualizationMode=Disabled
 volumeMultiplier=3.2319
 concurrencyOverrides {resolutionRule=PreventNew}
 concurrencySet [/Game/Sounds/Fort_Monster_Deimos/Settings/Deimos_Vocal_Global_Conc, /Game/Sounds/Fort_Monster_Deimos/Settings/Deimos_Vocal_Local_Conc]
}
