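# Sound cue for the pain vocalization of the ranged Fiend, a Deimos monster, routed to the tactical creature vocalization sound class.
# A mixer blends a close layer with a quieter distant layer, each with its own attenuation asset and modulator.
# Each layer randomly picks one of six equally weighted Fiend pain waves from its close or distant set.
# Global and local Deimos vocal concurrency sets govern playback, and virtualization is disabled.
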SoundCue /Game/Sounds/Fort_Monster_Deimos/Ranged/Fiend_Ranged_VO_Pain_Cue
root=Mixer_1
nodes [
 {id=Attenuation_2, children [Modulator_2], properties {attenuationSettings=/Game/Sounds/Attenuation/Fort_Deimos_Vocal_Close_Att}}
 {id=Attenuation_3, children [Modulator_3], properties {attenuationSettings=/Game/Sounds/Attenuation/Fort_Deimos_Vocal_Distant_Att}}
 {id=Mixer_1, children [Attenuation_2, Attenuation_3], properties {inputVolume [1.1,0.5]}}
 {id=Modulator_2, children [Random_1], properties {pitchMax=1.2}}
 {id=Modulator_3, children [Random_3]}
 {
  id=Random_1
  children [WavePlayer_5, WavePlayer_6, WavePlayer_7, WavePlayer_8, WavePlayer_9, WavePlayer_10]
  properties {bShouldExcludeFromBranchCulling=true, weights [1,1,1,1,1,1]}
 }
 {
  id=Random_3
  children [WavePlayer_18, WavePlayer_19, WavePlayer_20, WavePlayer_21, WavePlayer_22, WavePlayer_23]
  properties {bShouldExcludeFromBranchCulling=true, weights [1,1,1,1,1,1]}
 }
 {id=WavePlayer_10, properties {soundWaveAssetPtr=/Game/Sounds/Fort_Monster_Deimos/Fiend/Fiend_Pain_New_06}}
 {id=WavePlayer_18, properties {soundWaveAssetPtr=/Game/Sounds/Fort_Monster_Deimos/Fiend/Fiend_Pain_New_Distant_01}}
 {id=WavePlayer_19, properties {soundWaveAssetPtr=/Game/Sounds/Fort_Monster_Deimos/Fiend/Fiend_Pain_New_Distant_02}}
 {id=WavePlayer_20, properties {soundWaveAssetPtr=/Game/Sounds/Fort_Monster_Deimos/Fiend/Fiend_Pain_New_Distant_03}}
 {id=WavePlayer_21, properties {soundWaveAssetPtr=/Game/Sounds/Fort_Monster_Deimos/Fiend/Fiend_Pain_New_Distant_04}}
 {id=WavePlayer_22, properties {soundWaveAssetPtr=/Game/Sounds/Fort_Monster_Deimos/Fiend/Fiend_Pain_New_Distant_06}}
 {id=WavePlayer_23, properties {soundWaveAssetPtr=/Game/Sounds/Fort_Monster_Deimos/Fiend/Fiend_Pain_New_Distant_07}}
 {id=WavePlayer_5, properties {soundWaveAssetPtr=/Game/Sounds/Fort_Monster_Deimos/Fiend/Fiend_Pain_New_02}}
 {id=WavePlayer_6, properties {soundWaveAssetPtr=/Game/Sounds/Fort_Monster_Deimos/Fiend/Fiend_Pain_New_07}}
 {id=WavePlayer_7, properties {soundWaveAssetPtr=/Game/Sounds/Fort_Monster_Deimos/Fiend/Fiend_Pain_New_01}}
 {id=WavePlayer_8, properties {soundWaveAssetPtr=/Game/Sounds/Fort_Monster_Deimos/Fiend/Fiend_Pain_New_03}}
 {id=WavePlayer_9, properties {soundWaveAssetPtr=/Game/Sounds/Fort_Monster_Deimos/Fiend/Fiend_Pain_New_04}}
]
properties {
 bBypassVolumeScaleForPriority=true
 maxDistance=24000
 priority=0.75
 soundClassObject=&Fort_Creatures_Vocalization_Tactical
 virtualizationMode=Disabled
 volumeMultiplier=1.7955
 attenuationOverrides {distanceAlgorithm=NaturalSound, falloffDistance=4000, attenuationShapeExtents {x=1000}}
 concurrencyOverrides {resolutionRule=PreventNew}
 concurrencySet [/Game/Sounds/Fort_Monster_Deimos/Settings/Deimos_Vocal_Global_Conc, /Game/Sounds/Fort_Monster_Deimos/Settings/Deimos_Vocal_Local_Conc]
}
