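# Sound cue for the ranged Fiend's running breath vocalization, part of the Deimos monster sounds.
# A single Random node picks among twelve equally weighted breath wave variations.
# Output uses the creature vocalization sound class and the shared close-range Deimos vocal attenuation.
# Global and local Deimos vocal concurrency sets govern simultaneous playback, and virtualization is disabled.
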SoundCue /Game/Sounds/Fort_Monster_Deimos/Ranged/Fiend_Ranged_VO_Run_Breath_Cue
root=Random_0
nodes [
 {
  id=Random_0
  children [
   WavePlayer_0
   WavePlayer_1
   WavePlayer_2
   WavePlayer_3
   WavePlayer_4
   WavePlayer_9
   WavePlayer_10
   WavePlayer_11
   WavePlayer_12
   WavePlayer_13
   WavePlayer_14
   WavePlayer_15
  ]
  properties {bShouldExcludeFromBranchCulling=true, weights [1,1,1,1,1,1,1,1,1,1,1,1]}
 }
 {id=WavePlayer_0, properties {soundWaveAssetPtr=&/Fiend_Ranged_RunBreath_10}}
 {id=WavePlayer_1, properties {soundWaveAssetPtr=&/Fiend_Ranged_RunBreath_11}}
 {id=WavePlayer_10, properties {soundWaveAssetPtr=&/Fiend_Ranged_RunBreath_04}}
 {id=WavePlayer_11, properties {soundWaveAssetPtr=&/Fiend_Ranged_RunBreath_05}}
 {id=WavePlayer_12, properties {soundWaveAssetPtr=&/Fiend_Ranged_RunBreath_06}}
 {id=WavePlayer_13, properties {soundWaveAssetPtr=&/Fiend_Ranged_RunBreath_07}}
 {id=WavePlayer_14, properties {soundWaveAssetPtr=&/Fiend_Ranged_RunBreath_08}}
 {id=WavePlayer_15, properties {soundWaveAssetPtr=&/Fiend_Ranged_RunBreath_09}}
 {id=WavePlayer_2, properties {soundWaveAssetPtr=&/Fiend_Ranged_RunBreath_12}}
 {id=WavePlayer_3, properties {soundWaveAssetPtr=&/Fiend_Ranged_RunBreath_01}}
 {id=WavePlayer_4, properties {soundWaveAssetPtr=&/Fiend_Ranged_RunBreath_02}}
 {id=WavePlayer_9, properties {soundWaveAssetPtr=&/Fiend_Ranged_RunBreath_03}}
]
properties {
 attenuationSettings=/Game/Sounds/Attenuation/Fort_Deimos_Vocal_Close_Att
 bBypassVolumeScaleForPriority=true
 maxDistance=13000
 priority=0.75
 soundClassObject=/Game/Sounds/FortSoundClassesAndModes/Fort_Creatures_Vocalization_Normal
 virtualizationMode=Disabled
 volumeMultiplier=1.596
 attenuationOverrides {distanceAlgorithm=NaturalSound, falloffDistance=4000, attenuationShapeExtents {x=1000}}
 concurrencyOverrides {maxCount=2, volumeScale=0.75, volumeScaleAttackTime=0.1}
 concurrencySet [/Game/Sounds/Fort_Monster_Deimos/Settings/Deimos_Vocal_Global_Conc, /Game/Sounds/Fort_Monster_Deimos/Settings/Deimos_Vocal_Local_Conc]
}
